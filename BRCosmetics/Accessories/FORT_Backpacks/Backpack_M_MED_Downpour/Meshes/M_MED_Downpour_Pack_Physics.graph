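# Physics asset for the Downpour backpack cosmetic, adding simulated jiggle to its three spike bones.
# Each spike has a small capsule master body and a longer, offset capsule on its dynamic child bone.
# Each child is constrained to its own master with tight swing limits and almost no twist, so the spikes wobble slightly and independently.
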PhysicsAsset /BRCosmetics/Accessories/FORT_Backpacks/Backpack_M_MED_Downpour/Meshes/M_MED_Downpour_Pack_Physics
bodies {
 Dyn_Master_Spike1 [{shapeType=Capsule, length=0.5, radius=0.5, rotation=<rot 0 -90 0>}]
 Dyn_Master_Spike2 [{shapeType=Capsule, length=0.5, radius=0.5, rotation=<rot 0 -90 0>}]
 Dyn_Master_Spike3 [{shapeType=Capsule, length=0.5, radius=0.5, rotation=<rot 0 -90 0>}]
 Dyn_Spike1 [{shapeType=Capsule, length=2, radius=1, center=<1.5 0 0>, rotation=<rot 90 0 0>}]
 Dyn_Spike2 [{shapeType=Capsule, length=2, radius=1, center=<1.5 0 0>, rotation=<rot 90 0 0>}]
 Dyn_Spike3 [{shapeType=Capsule, length=2, radius=1, center=<1.5 0 0>, rotation=<rot 90 0 0>}]
}
constraints [
 {bone1Name=Dyn_Spike1, bone2Name=Dyn_Master_Spike1, swing1LimitDegrees=10, swing2LimitDegrees=20, twistLimitDegrees=1}
 {bone1Name=Dyn_Spike2, bone2Name=Dyn_Master_Spike2, swing1LimitDegrees=10, swing2LimitDegrees=20, twistLimitDegrees=1}
 {bone1Name=Dyn_Spike3, bone2Name=Dyn_Master_Spike3, swing1LimitDegrees=10, swing2LimitDegrees=20, twistLimitDegrees=1}
]
