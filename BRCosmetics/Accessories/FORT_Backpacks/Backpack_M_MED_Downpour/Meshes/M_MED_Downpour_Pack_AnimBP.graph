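# Animation Blueprint for the Downpour backpack, adding physics and procedural motion to its bones.
# AnimGraph runs RigidBody, then chains bone modifiers that offset the pack and apply AmbientRot/AmbientTrans variables to three spikes; Spike3 rotation targets Dyn_Spike2, and later translation nodes replace rotations with zero.
# A final pack offset and roll blends in while bIsSkydiveFloating is true.
# An Ambient Animation function computes time-based sine offsets, but nothing here calls it.
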AnimBlueprint /BRCosmetics/Accessories/FORT_Backpacks/Backpack_M_MED_Downpour/Meshes/M_MED_Downpour_Pack_AnimBP
target_skeleton=&/M_MED_Downpour_Pack_Skeleton
graphs [
 Ambient Animation {nodes [
   n1 = Ambient Animation!() -> (
    then
    Time Offset 1:double=~
    Magnitude_Multiplier:double=~
    Period1:Vector=~
    " Magnitude":Vector="10,0,0"
    Period Multiplier:double=1
    Rotate Multiplier:double=~
   ) {is=title K2Node_FunctionEntry}
   n2 = BreakVector(InVec=@n1." Magnitude")
   n3 = BreakVector(InVec=@n1.Period1)
   n4 = GetGameTimeInSeconds()
   n5 = float * float(A=float * float(A=@n1.Period Multiplier, B=@n3.Y), B=@n4)
   n6 = float * float(A="Sin (Radians)"(A=float + float(A=@n5, B=@n1) {is=title K2Node_CommutativeAssociativeBinaryOperator}), B=@n2.Y)
   n7 = float * float(A=float * float(A=@n1.Period Multiplier, B=@n3.Z), B=@n4)
   n8 = float * float(A="Sin (Radians)"(A=float + float(A=@n7, B=@n1)), B=@n2.Z)
   n9 = float * float(A=float * float(A=@n1.Period Multiplier, B=@n3), B=@n4)
   n10 = float * float(A="Sin (Radians)"(A=float + float(A=@n9, B=@n1)), B=@n2)
   n11 = MakeRotator(Roll=@n10, Pitch=@n6, Yaw=@n8)
   > Return Node!(
    execute
    VectorOut:Vector=vector * float(A=MakeVector(X=@n10, Y=@n6, Z=@n8), B=@n1.Magnitude_Multiplier)
    RotateOut:Rotator="Scale Rotator (Float)"!(A:Rotator="Scale Rotator (Float)"(A=@n11, B=@n1.Magnitude_Multiplier), B:float=@n1.Rotate Multiplier) -> (ReturnValue:Rotator)
   )
  ]}
 AnimGraph {nodes [
   Local Space Ref Pose()
   n1 = RigidBody(ComponentPose=Local To Component(LocalPose=Input Pose())) {
    bClampLinearTranslationLimitToRefPose=true
    bOverrideWorldGravity=true
    overridePhysicsAsset=&/M_MED_Downpour_Pack_Physics
    simulationSpace=BaseBoneSpace
    alphaBoolBlend {blendOption=Linear}
    baseBoneRef {boneName=root}
    componentAppliedLinearAccClamp=<1200 1200 1200>
    componentLinearAccScale=<0.1 0.1 0.1>
    componentLinearVelScale=<0.1 0.1 0.1>
    overrideWorldGravity {z=-980}
   }
   n2 = "Transform (Modify) Bone - Bone: pack"(ComponentPose=@n1, Translation="0,-2,0", Rotation="0,0,0") {
    translationMode=BMM_Additive
    translationSpace=BCS_BoneSpace
    alphaBoolBlend {blendOption=Linear}
    boneToModify {boneName=pack}
    translation {y=-2}
   }
   n3 = "Transform (Modify) Bone - Bone: Dyn_Spike1"(ComponentPose=@n2, Alpha=1, Translation="0,0,2", Rotation=Get AmbientRot_Spike1() -> (AmbientRot_Spike1:Rotator), Scale="1,1,1") {
    rotationMode=BMM_Replace
    rotationSpace=BCS_BoneSpace
    translationMode=BMM_Additive
    translationSpace=BCS_BoneSpace
    alphaBoolBlend {blendOption=Linear}
    boneToModify {boneName=Dyn_Spike1}
    translation {z=2}
   }
   n4 = "Transform (Modify) Bone - Bone: Dyn_Spike2"(ComponentPose=@n3, Alpha=1, Translation="0,0,0", Rotation=Get AmbientRot_Spike2() -> (AmbientRot_Spike2:Rotator), Scale="1,1,1") {
    rotationMode=BMM_Replace
    rotationSpace=BCS_BoneSpace
    translationMode=BMM_Additive
    translationSpace=BCS_BoneSpace
    alphaBoolBlend {blendOption=Linear}
    boneToModify {boneName=Dyn_Spike2}
   }
   n5 = "Transform (Modify) Bone - Bone: Dyn_Spike2"(ComponentPose=@n4, Alpha=1, Translation="0,0,0", Rotation=Get AmbientRot_Spike3() -> (AmbientRot_Spike3:Rotator), Scale="1,1,1") {
    rotationMode=BMM_Replace
    rotationSpace=BCS_BoneSpace
    translationMode=BMM_Additive
    translationSpace=BCS_BoneSpace
    alphaBoolBlend {blendOption=Linear}
    boneToModify {boneName=Dyn_Spike2}
   }
   n6 = "Transform (Modify) Bone - Bone: Dyn_Spike1"(ComponentPose=@n5, Alpha=1, Translation=Get AmbientTrans_Spike1() -> (AmbientTrans_Spike1:Vector), Rotation="0,0,0", Scale="1,1,1") {
    rotationMode=BMM_Replace
    rotationSpace=BCS_BoneSpace
    translationMode=BMM_Additive
    translationSpace=BCS_BoneSpace
    alphaBoolBlend {blendOption=Linear}
    boneToModify {boneName=Dyn_Spike1}
   }
   n7 = "Transform (Modify) Bone - Bone: Dyn_Spike2"(ComponentPose=@n6, Alpha=1, Translation=Get AmbientTrans_Spike2() -> (AmbientTrans_Spike2:Vector), Rotation="0,0,0", Scale="1,1,1") {
    rotationMode=BMM_Replace
    rotationSpace=BCS_BoneSpace
    translationMode=BMM_Additive
    translationSpace=BCS_BoneSpace
    alphaBoolBlend {blendOption=Linear}
    boneToModify {boneName=Dyn_Spike2}
   }
   n8 = "Transform (Modify) Bone - Bone: Dyn_Spike3"(ComponentPose=@n7, Alpha=1, Translation=Get AmbientTrans_Spike3() -> (AmbientTrans_Spike3:Vector), Rotation="0,0,0", Scale="1,1,1") {
    rotationMode=BMM_Replace
    rotationSpace=BCS_BoneSpace
    translationMode=BMM_Additive
    translationSpace=BCS_BoneSpace
    alphaBoolBlend {blendOption=Linear}
    boneToModify {boneName=Dyn_Spike3}
   }
   n9 = "Transform (Modify) Bone - Bone: pack"!(ComponentPose=@n8, bAlphaBoolEnabled:bool=Get bIsSkydiveFloating(), Translation="0,-10,0", Rotation="0,0,30", Scale) -> (Pose) {
    alphaInputType=Bool
    rotationMode=BMM_Additive
    rotationSpace=BCS_BoneSpace
    translationMode=BMM_Additive
    translationSpace=BCS_BoneSpace
    alphaBoolBlend {blendInTime=5, blendOption=Linear, blendOutTime=5}
    boneToModify {boneName=pack}
    rotation {roll=30}
    translation {y=-10}
   }
   Output Pose(Result=Component To Local(ComponentPose=@n9)) {layerGroup=None}
  ]}
 EventGraph {links [], nodes []}
]
variables []
